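# Emissive material for Venue_A stage geometry, showing animated caustics tinted by the stage's Tertiary color and pulsing with the beat.
# A reflection-vector parallax offset gives the EmissiveTexture's red channel fake depth, multiplied by its undistorted green channel and caustics.
# Two rotating, panning caustic samples layer, one distorting the other; NoMask blends toward unmasked caustics.
# MF_MojoBeatPulse and MF_FX_EmissiveAdjustments process the result; the StageColor lerp has Tertiary on both inputs.
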Material /SparksMaps/Environment/Venue_A/Materials/M_Stage_Emissive
graph {
 properties {bUsedWithInstancedStaticMeshes=true, bUsedWithNanite=true, pixelDepthOffsetMode=PDOM_Legacy, usedShadingModels=Default Lit}
 nodes [
  Constant()
  Constant()
  Constant() {r=0.35}
  Constant() {r=0.7}
  n1 = ScalarParameter() {defaultValue=1, parameterName=UVTile_Depth}
  n2 = TextureCoordinate()
  n3 = MaterialFunctionCall!(
   Normal=MaterialFunctionCall(R=Multiply() {constB=0.1}, G=Multiply() {constB=0.1}, B=Constant() {r=1}) {materialFunction=&AppendMany}.RGB
   CameraVector=Transform(None=CameraVectorWS()) {transformSourceType=TRANSFORMSOURCE_World, transformType=TRANSFORM_Tangent}
  ) {materialFunction=/Engine/Functions/Engine_MaterialFunctions02/Utility/CustomReflectionVector}
  n4 = ComponentMask(None=@n3.Reflection Vector) {b=true}
  n5 = Multiply(
   A=ComponentMask(None=@n3.Reflection Vector) {g=true, r=true}
   B=Divide(A=ScalarParameter() {defaultValue=10, parameterName=Depth}, B=Abs(None=@n4))
  )
  n6 = Divide(A=Constant() {r=1}, B=Constant() {r=1024})
  n7 = TextureObjectParameter() {parameterName=EmissiveTexture, texture=/SparksMaps/Environment/Venue_A/Textures/T_EmisTrim_A_E}
  n8 = TextureSample(UVs=Add(A=Multiply(A=@n1, B=@n2), B=Multiply(A=@n5, B=@n6)), Tex=@n7) {texture=/Game/Environments/Galileo/Textures/T_Galileo_Stage_Cylinder_E}
  n9 = ScalarParameter() {defaultValue=1, parameterName=CausticTilign}
  n10 = Rotator(Coordinate=Multiply(A=TextureCoordinate() {uTiling=0.5, unMirrorU=true, vTiling=0.5}, B=@n9)) {centerX=0.25, centerY=0.25, speed=0.01}
  n11 = Multiply(A=TextureCoordinate() {uTiling=0.35, unMirrorU=true, vTiling=0.35}, B=@n9)
  n12 = TextureSample(UVs=Panner(Coordinate=Rotator(Coordinate=@n11) {speed=0.05}) {speedX=-0.01, speedY=0.02}, Tex=NamedRerouteUsage())
  n13 = TextureSample(UVs=Add(A=Panner(Coordinate=@n10) {speedX=0.02}, B=Multiply(A=@n12.G) {constB=0.03}), Tex=NamedRerouteUsage())
  n14 = Multiply(A=TextureSample(Tex=@n7) {texture=&T_Galileo_Stage_Cylinder_E}.G, B=@n13.RGB)
  n15 = Multiply(A=Constant() {r=5}, B=LinearInterpolate(A=Multiply(A=@n8.R, B=@n14), B=@n13.RGB, Alpha=ScalarParameter() {parameterName=NoMask}))
  n16 = MaterialFunctionCall!(CustomUPosition) {materialFunction=/SparksStageShow/Systems/StageManager/FXManager/MaterialFunctions/MF_MojoStageColors}
  n17 = LinearInterpolate(A=@n16.Tertiary, B=@n16.Tertiary, Alpha=ScalarParameter() {parameterName=StageColor}) {constAlpha=0}
  MaterialFunctionCall(Emissive=MaterialFunctionCall!(Emissive=Multiply(A=@n15, B=@n17), BeatPulse Strength, BeatPulse Min, BeatPulse Max) {materialFunction=/SparksMaps/Environment/MaterialFunctions/MF_MojoBeatPulse}.Result) {materialFunction=&MF_FX_EmissiveAdjustments}
  NamedRerouteDeclaration(None=TextureObject() {texture=/Game/Effects/Fort_Effects/Textures/Tile/T_Caustics_Chromatic}) {name=TxrCaustic, nodeColor=<rgba 0 1 0.9531 1>}
 ]
}
