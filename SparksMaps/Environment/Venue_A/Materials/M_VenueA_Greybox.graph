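# M_VenueA_Greybox is a simple Default Lit material, named as the greybox surface for the Venue A environment in SparksMaps.
# An RGBTINT vector parameter is multiplied by a scalar Color parameter to give an adjustable tinted brightness.
# Separate Metal, Spec and Rough scalar parameters expose uniform surface values, with no textures in the graph.
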Material /SparksMaps/Environment/Venue_A/Materials/M_VenueA_Greybox
graph {
 properties {pixelDepthOffsetMode=PDOM_Legacy, usedShadingModels=Default Lit}
 nodes [
  Multiply(A=VectorParameter() {parameterName=RGBTINT, defaultValue=<rgba 1 1 1 1>}.RGB, B=ScalarParameter() {defaultValue=0.5, parameterName=Color})
  ScalarParameter() {defaultValue=1, parameterName=Metal}
  ScalarParameter() {defaultValue=0.5, parameterName=Spec}
  ScalarParameter() {defaultValue=0.35, parameterName=Rough}
 ]
}
