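# Lit Venue A material that puts animated emissive rain over corrugated metal textures, usable on Nanite and instanced meshes.
# Three channels of one rain texture pan vertically at two speeds, pass through overlay blends, and are weighted via named reroutes for stage colors, inverted Fresnel and strength.
# A smoothstepped height gradient, normalized by object bounds, is added.
# The sum feeds the MojoBeatPulse and emissive adjustment functions.
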Material /SparksMaps/Environment/Venue_A/Materials/M_Venue_A_RainEmis
graph {
 properties {bUsedWithInstancedStaticMeshes=true, bUsedWithNanite=true, pixelDepthOffsetMode=PDOM_Legacy, usedShadingModels=Default Lit}
 nodes [
  n1 = TextureCoordinate() {uTiling=0.25, vTiling=0.15}
  n2 = Panner(Coordinate=@n1) {speedY=-0.025}
  n3 = Multiply(
   A=MaterialFunctionCall(
    Base=TextureSampleParameter2D(UVs=@n2) {group=Rainfall, parameterName=RainTexture, texture=/SparksMaps/Environment/Venue_A/Textures/T_EmissiveRain}.G
    Blend=NamedRerouteUsage()
   ) {materialFunction=&Blend_Overlay}.Result
   B=NamedRerouteUsage()
  )
  n4 = NamedRerouteUsage()
  n5 = Panner(Coordinate=@n1) {speedY=-0.1}
  n6 = Multiply(
   A=MaterialFunctionCall(
    Base=TextureSampleParameter2D(UVs=@n5) {group=Rainfall, parameterName=RainTexture, texture=&T_EmissiveRain}.R
    Blend=NamedRerouteUsage()
   ) {materialFunction=&Blend_Overlay}.Result
   B=NamedRerouteUsage()
  )
  n7 = NamedRerouteUsage()
  n8 = MaterialFunctionCall(
   Base=TextureSampleParameter2D(UVs=Panner(Coordinate=@n1) {speedY=-0.1}) {group=Rainfall, parameterName=RainTexture, texture=&T_EmissiveRain}.B
   Blend=NamedRerouteUsage()
  ) {materialFunction=&Blend_Overlay}
  n9 = NamedRerouteUsage()
  n10 = Add(A=Add(A=Multiply(A=@n3, B=@n4), B=Multiply(A=@n6, B=@n7)), B=Multiply(A=Multiply(A=@n8.Result, B=@n9), B=NamedRerouteUsage()))
  n11 = ScalarParameter() {defaultValue=0.5, group=Gradient, parameterName=Smooth Min}
  n12 = ScalarParameter() {defaultValue=0.9, group=Gradient, parameterName=Smooth Max}
  n13 = WorldPosition()
  n14 = ObjectPositionWS()
  n15 = Divide(
   A=Subtract(
    A=TransformPosition(None=@n13.XYZ) {transformType=TRANSFORMPOSSOURCE_World}
    B=TransformPosition(None=@n14) {transformType=TRANSFORMPOSSOURCE_World}
   )
   B=ObjectLocalBounds().Max
  )
  n16 = Power(Base=ComponentMask(None=Add(A=@n15)) {b=true}, Exp=ScalarParameter() {defaultValue=0.4, group=Gradient, parameterName=Power Exponent})
  n17 = Multiply(A=NamedRerouteUsage(), B=OneMinus(None=SmoothStep(Min=@n11, Max=@n12, Value=@n16) {constValue=1}))
  MaterialFunctionCall(Emissive=MaterialFunctionCall!(Emissive=Add(A=@n10, B=@n17), BeatPulse Strength, BeatPulse Min, BeatPulse Max) {materialFunction=/SparksMaps/Environment/MaterialFunctions/MF_MojoBeatPulse}.Result) {materialFunction=&MF_FX_EmissiveAdjustments}
  n18 = MaterialFunctionCall!(CustomUPosition) {materialFunction=/SparksStageShow/Systems/StageManager/FXManager/MaterialFunctions/MF_MojoStageColors}
  NamedRerouteDeclaration(None=@n18.Secondary) {name=ColorSecondary, nodeColor=<rgba 0 0.7074 1 1>}
  NamedRerouteDeclaration(None=@n18.Tertiary) {name=ColorTertiary, nodeColor=<rgba 0 1 0.25 1>}
  NamedRerouteDeclaration(None=@n18.Primary) {name=ColorPrimary, nodeColor=<rgba 0.4531 0 1 1>}
  n19 = ScalarParameter() {defaultValue=0.1, group=Rainfall, parameterName=Fresnel Strength}
  NamedRerouteDeclaration(None=OneMinus(None=Fresnel(ExponentIn=@n19))) {name=FresnelInput, nodeColor=<rgba 0 0.2031 1 1>}
  NamedRerouteDeclaration(None=ScalarParameter() {defaultValue=10, group=Rainfall, parameterName=Emissive Strength}) {name=EmissiveInput, nodeColor=<rgba 0.5156 0.4866 0.4297 1>}
  ScalarParameter() {defaultValue=0.7, group=Base Textures, parameterName=Metallic}
  TextureSampleParameter2D() {group=Base Textures, parameterName=Albedo, texture=/SparksMaps/Environment/Venue_A/Textures/T_Corrugated_Metal_A_D}
  TextureSampleParameter2D(UVs=TextureCoordinate() {uTiling=0.3, vTiling=0.3}) {group=Base Textures, parameterName=Roughness Variation, samplerType=SAMPLERTYPE_Masks, texture=/FortWater/Textures/OceanNoise__Mask}
  TextureSampleParameter2D() {
   group=Base Textures
   parameterName=Normal
   samplerType=SAMPLERTYPE_Normal
   texture=/SparksMaps/Environment/Venue_A/Textures/T_Corrugated_Metal_A_N
  }
  TextureSample() {samplerType=SAMPLERTYPE_LinearColor, texture=/SparksMaps/Environment/Venue_A/Textures/T_Corrugated_Metal_A_S}
 ]
}
